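# Material function that overlays holographic scanlines on an input color for the Season 30 player Superlevel effects.
# It feeds scaled UVs, by default stabilized screen-space UVs, into the shared holographic scanline function and sharpens the result with a thickness-driven SmoothStep.
# That mask blends between brightened and darkened copies of the input, set by two intensity inputs, and the banded result is mixed over the original at a fixed weight.
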MaterialFunction /Game/Characters/Player/TechArt/Superlevels/S30/MaterialFunctions/MF_SL_S30_Scanlines
graph {nodes [
  n1 = FunctionInput(Preview=TextureSample() {texture=/Game/Athena/UI/InGame/Sidekick/Reactions/100}.RGB)
  n2 = Multiply(
   A=@n1
   B=FunctionInput(Preview=Constant() {r=1.75}) {bUsePreviewValueAsDefault=true, inputName=ScanlineIntensity, inputType=FunctionInput_Scalar}
  )
  n3 = Multiply(
   A=@n1
   B=FunctionInput(Preview=Constant() {r=0.5}) {bUsePreviewValueAsDefault=true, inputName=ScanNegativeIntensity, inputType=FunctionInput_Scalar}
  )
  n4 = FunctionInput(Preview=MaterialFunctionCall!(ObjectPivotLocation, ObjectPositionOffset=Constant3Vector() {constant=<rgba 0 0 100 1>}) {materialFunction=/Game/Characters/Player/Common/MaterialFunctions/Superlevel/ImprovedStabilizedScreenSpaceUVs}.StabilizedUVs) {bUsePreviewValueAsDefault=true, description=Default is ScreenSpace, inputName=UVs, inputType=FunctionInput_Vector2}
  n5 = FunctionInput(Preview=Constant() {r=4}) {bUsePreviewValueAsDefault=true, inputName=ScanlineScale, inputType=FunctionInput_Scalar}
  n6 = SmoothStep(
   Min=FunctionInput(Preview=Constant()) {bUsePreviewValueAsDefault=true, inputName=ScanlineThickness, inputType=FunctionInput_Scalar}
   Value=MaterialFunctionCall!(
    ScanLine_Tiling_1
    Scanline_PannerSpeed_1
    ScanLine_Tiling
    Scanline_PannerSpeed
    Scanline_Rotation
    UVs=Multiply(A=@n4, B=@n5) {constB=0.04}
   ) {desc=ScanLine, materialFunction=/Game/Global/Materials/MaterialGym/Materials/HolographicTexture/MF_HolographicTexture_ScanLines}.Result
  ) {constMin=1}
  FunctionOutput(None=LinearInterpolate(A=@n1, B=LinearInterpolate(A=@n2, B=@n3, Alpha=@n6)) {constAlpha=0.75})
 ]}
