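# Material function for the S30 player Superlevel effect that outputs an animated Squares mask from a packed texture.
# R and B offset the timing of two layers, whose G and A values are shifted by clamped time sines, then thresholded through sine and smoothstep.
# The line threshold blends from near to far by camera depth, and a parameter weights the second layer before summing.
# UVs default to scaled, pivot-stabilized screen-space coordinates.
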MaterialFunction /Game/Characters/Player/TechArt/Superlevels/S30/MaterialFunctions/MF_SL_S30_Squares
graph {nodes [
  n1 = LinearInterpolate(
   A=ScalarParameter() {defaultValue=0.96, group=SL_S30_Squares, parameterName=SQ_LineThicknessNear}
   B=ScalarParameter() {defaultValue=0.95, group=SL_S30_Squares, parameterName=SQ_LineThicknessFar}
   Alpha=MaterialFunctionCall(Fade Length=Constant() {r=50}, Fade Offset=Constant() {r=100}) {materialFunction=&CameraDepthFade}.Result
  )
  n2 = MaterialFunctionCall!(ObjectPivotLocation, ObjectPositionOffset=Constant3Vector() {constant=<rgba 0 0 110 1>}) {materialFunction=/Game/Characters/Player/Common/MaterialFunctions/Superlevel/ImprovedStabilizedScreenSpaceUVs}
  n3 = TextureSample(UVs=FunctionInput(Preview=Multiply(A=@n2.RawUVs) {constB=0.015}) {bUsePreviewValueAsDefault=true, inputName=UV, inputType=FunctionInput_Vector2}) {samplerType=SAMPLERTYPE_LinearColor, texture=/Game/Characters/Player/TechArt/Superlevels/S30/Textures/T_S30_Superlevel_T3_Squares}
  n4 = Saturate(None=Subtract(A=@n3.G, B=Multiply(A=@n3.R) {constB=0.5}))
  n5 = Multiply(A=Subtract(A=Time(), B=Multiply(A=@n3.R) {constB=10})) {constB=0.5}
  n6 = Multiply(A=Saturate(None=Sine(None=@n5))) {constB=0.6}
  n7 = SmoothStep(Min=@n1, Value=Sine(None=Subtract(A=@n4, B=@n6))) {constMin=0.95}
  n8 = Time()
  n9 = Multiply(A=@n3.B) {constB=10}
  n10 = Sine(None=Multiply(A=Subtract(A=@n8, B=@n9)) {constB=0.4})
  n11 = Subtract(A=@n3.A, B=Multiply(A=Saturate(None=@n10)) {constB=0.6})
  n12 = Multiply(
   A=SmoothStep(Min=@n1, Value=Sine(None=@n11)) {constMin=0.95}
   B=ScalarParameter() {defaultValue=1, group=SL_S30_Squares, parameterName=SQ_UseTwoLayers, sliderMax=1}
  )
  FunctionOutput(None=NamedRerouteDeclaration(None=Add(A=@n7, B=@n12)) {name=Squares, nodeColor=<rgba 1 0.6094 0 1>})
 ]}
