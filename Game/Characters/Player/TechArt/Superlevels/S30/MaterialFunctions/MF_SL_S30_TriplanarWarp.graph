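# Material function for Season 30 player Superlevel materials that projects a scrolling, noise-warped triplanar texture in pre-skinned local space, so it stays fixed to the animated mesh.
# Local position, scaled and panned over time, yields three planar UV sets, with samples blended by contrast-adjusted normal masks.
# A separate noise texture, projected on local XZ with a selectable channel, distorts the UVs.
# Outputs are the three-plane blend, two-plane blend and noise value.
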MaterialFunction /Game/Characters/Player/TechArt/Superlevels/S30/MaterialFunctions/MF_SL_S30_TriplanarWarp
graph {nodes [
  n1 = PreSkinnedPosition()
  n2 = Constant() {r=60}
  n3 = Constant() {r=1}
  n4 = Divide(
   A=FunctionInput(Preview=VertexInterpolator(VS=@n1).PS) {bUsePreviewValueAsDefault=true, inputName=Local Position}
   B=Multiply(
    A=NamedRerouteDeclaration(None=@n2) {name=TriplanarScale, nodeColor=<rgba 0.4297 0 1 1>}
    B=FunctionInput(Preview=@n3) {bUsePreviewValueAsDefault=true, inputName=TriplannarTextureScale, inputType=FunctionInput_Scalar}
   )
  )
  n5 = Time()
  n6 = FunctionInput(Preview=Constant3Vector() {constant=<rgba 0 0.25 0.25 1>}) {inputName=Triplanner Speed}
  n7 = MaterialFunctionCall(Color=Frac(None=Multiply(A=@n5, B=@n6))) {materialFunction=&SplitComponents}
  n8 = Add(A=ComponentMask(None=@n4) {b=true, g=true}, B=AppendVector(A=@n7.R, B=Multiply(A=@n7.B) {constB=-1}))
  n9 = Multiply(
   A=Subtract(A=NamedRerouteUsage()) {constB=0.5}
   B=FunctionInput(Preview=Constant() {r=0.3}) {bUsePreviewValueAsDefault=true, inputName=NoiseDeformIntensity, inputType=FunctionInput_Scalar}
  )
  n10 = FunctionInput(Preview=TextureObject() {texture=/Game/Characters/Player/TechArt/Superlevels/S30/Textures/T_S30_Superlevel_T3_Curves}) {inputName=MainTexture, inputType=FunctionInput_Texture2D}
  n11 = Add(A=ComponentMask(None=@n4) {b=true, r=true}, B=AppendVector(A=@n7.R, B=@n7.R)) {constB=0}
  n12 = LinearInterpolate(
   A=TextureSample(UVs=Add(A=@n8, B=@n9), Tex=@n10) {samplerType=SAMPLERTYPE_Masks, texture=/Game/Weapons/WeaponSkins/Textures/T_Swirly_Noise_highContrast}.RGBA
   B=TextureSample(UVs=Add(A=@n11, B=@n9), Tex=@n10) {samplerType=SAMPLERTYPE_Masks, texture=&T_Swirly_Noise_highContrast}.RGBA
   Alpha=NamedRerouteUsage()
  )
  n13 = Add(A=ComponentMask(None=@n4) {g=true, r=true}, B=AppendVector(A=@n7.R, B=@n7.R)) {constB=0}
  n14 = LinearInterpolate(
   A=@n12
   B=TextureSample(UVs=Add(A=@n13, B=@n9), Tex=@n10) {samplerType=SAMPLERTYPE_Masks, texture=&T_Swirly_Noise_highContrast}.RGBA
   Alpha=NamedRerouteUsage()
  )
  ComponentMask(None=@n14) {a=true}
  FunctionOutput(None=@n14) {outputName=XYZ}
  FunctionOutput(None=@n12) {outputName=XY}
  n15 = Divide(A=ComponentMask(None=PreSkinnedPosition()) {b=true, r=true}, B=NamedRerouteUsage())
  n16 = FunctionInput(Preview=Constant() {r=1}) {bUsePreviewValueAsDefault=true, inputName=NoiseTiling, inputType=FunctionInput_Scalar}
  n17 = TextureSample(
   UVs=VertexInterpolator(VS=Multiply(A=@n15, B=@n16)).PS
   Tex=FunctionInput(Preview=TextureObject() {texture=&T_S30_Superlevel_T3_Curves}) {inputName=NoiseTexture, inputType=FunctionInput_Texture2D}
  ) {texture=&T_S30_Superlevel_T3_Curves}
  n18 = FunctionInput(Preview=Constant4Vector() {constant=<rgba 0 1 0 0>}.RGBA) {bUsePreviewValueAsDefault=true, inputName=NoiseChannel, inputType=FunctionInput_Vector4}
  FunctionOutput(None=NamedRerouteDeclaration(None=DotProduct(A=@n17.RGBA, B=@n18)) {name=NoiseTexture, nodeColor=<rgba 0.2891 1 0 1>}) {outputName=NoiseTexture, sortPriority=1}
  MaterialFunctionCall!("Normal (WS)", Texture Scale, Triplanar Fade, Derivatives Scale) {materialFunction=/Game/Athena/Environments/MaterialFunctions/MF_TriplanarDithered}
  n19 = FunctionInput(Preview=VertexInterpolator(VS=PreSkinnedNormal()).PS) {bUsePreviewValueAsDefault=true, inputName=Local Normal}
  n20 = FunctionInput(Preview=Constant() {r=1}) {bUsePreviewValueAsDefault=true, inputName=TriPlanner AxisBlendStrength, inputType=FunctionInput_Scalar}
  n21 = MaterialFunctionCall(In=Abs(None=ComponentMask(None=@n19) {g=true}), Contrast=@n20) {materialFunction=&CheapContrast}
  NamedRerouteDeclaration(None=Saturate(None=ComponentMask(None=@n21.Result) {g=true})) {name=FrontMask, nodeColor=<rgba 0 0.9297 1 1>}
  n22 = MaterialFunctionCall(In=Abs(None=ComponentMask(None=@n19) {b=true}), Contrast=@n20) {materialFunction=&CheapContrast}
  NamedRerouteDeclaration(None=Saturate(None=ComponentMask(None=@n22.Result) {b=true})) {name=TopMask, nodeColor=<rgba 0 0.9297 1 1>}
 ]}
